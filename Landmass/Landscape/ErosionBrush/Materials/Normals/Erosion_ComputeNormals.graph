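# Unlit AlphaComposite material for the Landmass erosion brush that computes terrain and water-surface normals from HeightRT.
# A Custom node central-differences four texel-snapped neighbours, scaled by MPC_Erosion resolution, GridSize and ZScale.
# It derives XY normals for height (R) and height plus water (read from G, though its comment says B).
# Both pairs are bias-scaled and packed into RGBA, alpha inverted; the ShowWater input is unused by the code.
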Material /Landmass/Landscape/ErosionBrush/Materials/Normals/Erosion_ComputeNormals
graph {
 properties {
  bAllowNegativeEmissiveColor=true
  bTangentSpaceNormal=false
  blendMode=BLEND_AlphaComposite
  pixelDepthOffsetMode=PDOM_Legacy
  shadingModel=MSM_Unlit
  usedShadingModels=Unlit
 }
 nodes [
  MaterialExpressionTextureObjectParameter_1 = TextureObjectParameter() {parameterName=HeightRT, texture=&Black}
  MaterialExpressionAppendVector_1 = AppendVector(
   A=CollectionParameter() {collection=/Landmass/Landscape/BlueprintBrushes/MPC/MPC_Erosion, parameterName=ResX}
   B=CollectionParameter() {collection=&MPC_Erosion, parameterName=ResY}
  )
  MaterialExpressionTextureCoordinate_1 = TextureCoordinate()
  MaterialExpressionCollectionParameter_0 = CollectionParameter() {collection=&MPC_Erosion, parameterName=GridSize}
  MaterialExpressionCollectionParameter_3 = CollectionParameter() {collection=&MPC_Erosion, parameterName=ZScale}
  MaterialExpressionAppendVector_3 = AppendVector(A=Constant() {r=1}, B=ScalarParameter() {defaultValue=1, parameterName=ShowWater})
  n1 = Custom!(
   HeightTex=@MaterialExpressionTextureObjectParameter_1
   TextureSize=@MaterialExpressionAppendVector_1
   UV=@MaterialExpressionTextureCoordinate_1
   GridSize=@MaterialExpressionCollectionParameter_0
   HeightScale=@MaterialExpressionCollectionParameter_3
   D=@MaterialExpressionAppendVector_3
  ) {
   code="//HeightTex: input Texture Object containing heights.R and water.B\r\n//TextureSize: input for size\r\n//GridSize: input float grid world size\r\n//HeightScale: input for height range scale\r\n\r\nfloat2 offsets[4] = {float2(0,1), float2(0,-1), float2(1,0), float2(-1,0)};\r\n\r\nfloat2 texelsize = 1 / TextureSize;\r\n\r\nUV = (floor(UV * TextureSize)/TextureSize) + (texelsize * 0.5);\r\n\r\nfloat2 PosYSample = HeightTex.SampleLevel(HeightTexSampler, UV + offsets[0] * texelsize, 0).rg;\r\nfloat2 NegYSample = HeightTex.SampleLevel(HeightTexSampler, UV + offsets[1] * texelsize, 0).rg;\r\nfloat2 PosXSample = HeightTex.SampleLevel(HeightTexSampler, UV + offsets[2] * texelsize, 0).rg;\r\nfloat2 NegXSample = HeightTex.SampleLevel(HeightTexSampler, UV + offsets[3] * texelsize, 0).rg;\r\n\r\nfloat3 tangent =  float3( 2*GridSize, 0, ((PosXSample.x) - (NegXSample.x))*HeightScale);\r\nfloat3 binormal = float3( 0, 2*GridSize, ((PosYSample.x) - (NegYSample.x))*HeightScale);\r\ntangent = normalize(tangent);\r\nbinormal = normalize(binormal);\r\n\r\nfloat3 normal1 = normalize(cross(tangent, binormal));\r\n\r\n//float2 normal1 = float2(-tangent.z, -binormal.z);\r\n\r\nPosYSample.x +=  PosYSample.y;\r\nNegYSample.x +=  NegYSample.y;\r\nPosXSample.x +=  PosXSample.y;\r\nNegXSample.x +=  NegXSample.y;\r\n\r\ntangent =  float3( 2*GridSize, 0, ((PosXSample.x) - (NegXSample.x))*HeightScale);\r\nbinormal = float3( 0, 2*GridSize, ((PosYSample.x) - (NegYSample.x))*HeightScale);\r\ntangent = normalize(tangent);\r\nbinormal = normalize(binormal);\r\n\r\nfloat3 normal2 = normalize(cross(tangent, binormal));\r\n\r\nreturn float4(normal1.x, normal1.y, normal2.x, normal2.y);"
   description=Compute Normal
   outputType=CMOT_Float4
   inputs [
    {
     inputName=HeightTex
     input {expression=&self:MaterialExpressionTextureObjectParameter_1}
    }
    {
     inputName=TextureSize
     input {expression=&self:MaterialExpressionAppendVector_1}
    }
    {
     inputName=UV
     input {expression=&self:MaterialExpressionTextureCoordinate_1}
    }
    {
     inputName=GridSize
     input {expression=&self:MaterialExpressionCollectionParameter_0}
    }
    {
     inputName=HeightScale
     input {expression=&self:MaterialExpressionCollectionParameter_3}
    }
    {inputName=D, input {expression=&self:MaterialExpressionAppendVector_3}}
   ]
  }
  n2 = ConstantBiasScale(None=@n1)
  ComponentMask(None=@n2) {b=true, g=true, r=true}
  Constant() {desc=Erosion height is in World Space already, r=1}
  OneMinus(None=ComponentMask(None=@n2) {a=true})
 ]
}
